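# Sound cue for an oak tree being destroyed, routed to the Fort_Impacts sound class.
# A quality-level root sends its first two slots to a mixer and its last two to a single pitch-varied Tree_Destroy_opt wave.
# The mixer layers two random picks, one of tree-final impacts and one of wood breaks, each with pitch modulation.
# A shared object-impact concurrency set governs simultaneous playback.
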SoundCue /Clyde/Sounds/Fort_Impact_Sounds/Demolition/Wood/Demo_OakTree_Destroyed_Cue
root=QualityLevel_0
nodes [
 {id=Mixer_2, children [Modulator_1,Modulator_4], properties {inputVolume [1,1]}}
 {id=Modulator_0, children [WavePlayer_0], properties {pitchMax=1.1, pitchMin=0.9}}
 {id=Modulator_1, children [Random_7], properties {pitchMax=1, pitchMin=0.9}}
 {id=Modulator_4, children [Random_5], properties {pitchMax=1.1, pitchMin=0.9}}
 {id=QualityLevel_0, children [Mixer_2,Mixer_2,Modulator_0,Modulator_0]}
 {id=Random_5, children [WavePlayer_23, WavePlayer_24], properties {weights [1,1]}}
 {id=Random_7, children [WavePlayer_42, WavePlayer_43], properties {weights [1,1]}}
 {id=WavePlayer_0, properties {soundWaveAssetPtr=&/Tree_Destroy_opt}}
 {id=WavePlayer_23, properties {soundWaveAssetPtr=/Game/Sounds/Fort_Impact_Sounds/Demolition/Wood/final_wood_new_03}}
 {id=WavePlayer_24, properties {soundWaveAssetPtr=/Game/Sounds/Fort_Impact_Sounds/Demolition/Wood/final_wood_new_02}}
 {id=WavePlayer_42, properties {soundWaveAssetPtr=/Clyde/Sounds/Fort_Impact_Sounds/Demolition/Fort_Impact_TreeFinal_02}}
 {id=WavePlayer_43, properties {soundWaveAssetPtr=/Clyde/Sounds/Fort_Impact_Sounds/Demolition/Fort_Impact_TreeFinal_01}}
]
properties {
 maxDistance=2097152
 soundClassObject=&Fort_Impacts
 volumeMultiplier=0.272
 concurrencyOverrides {resolutionRule=PreventNew}
 concurrencySet [&CONC_Object_Impacts_FourMax]
}
