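# Sound cue for oak tree impacts in wood demolition, playing big tree hit sounds in the impacts sound class.
# A quality level node selects among four branches with progressively fewer variants, from four equally weighted random waves down to a single wave.
# Each branch passes through a modulator that randomly lowers pitch, with a wider range on the single-wave branch.
# A shared object-impact concurrency set limits simultaneous playback.
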SoundCue /Clyde/Sounds/Fort_Impact_Sounds/Demolition/Wood/Demo_OakTree_Impact_Cue
root=QualityLevel_0
nodes [
 {id=Modulator_0, children [Random_0], properties {pitchMax=0.9, pitchMin=0.7}}
 {id=Modulator_1, children [Random_1], properties {pitchMax=0.9, pitchMin=0.7}}
 {id=Modulator_2, children [Random_6], properties {pitchMax=0.9, pitchMin=0.7}}
 {id=Modulator_3, children [WavePlayer_6], properties {pitchMax=1, pitchMin=0.7}}
 {id=QualityLevel_0, children [Modulator_2,Modulator_0,Modulator_1,Modulator_3]}
 {id=Random_0, children [WavePlayer_0,WavePlayer_2,WavePlayer_1], properties {weights [1,1,1]}}
 {id=Random_1, children [WavePlayer_4,WavePlayer_5], properties {weights [1,1]}}
 {id=Random_6, children [WavePlayer_35, WavePlayer_38, WavePlayer_37, WavePlayer_39], properties {weights [1,1,1,1]}}
 {id=WavePlayer_0, properties {soundWaveAssetPtr=&/tree_big_hit_01}}
 {id=WavePlayer_1, properties {soundWaveAssetPtr=&/tree_big_hit_03}}
 {id=WavePlayer_2, properties {soundWaveAssetPtr=&/tree_big_hit_07}}
 {id=WavePlayer_35, properties {soundWaveAssetPtr=&tree_big_hit_01}}
 {id=WavePlayer_37, properties {soundWaveAssetPtr=&tree_big_hit_03}}
 {id=WavePlayer_38, properties {soundWaveAssetPtr=&tree_big_hit_07}}
 {id=WavePlayer_39, properties {soundWaveAssetPtr=&/tree_big_hit_08}}
 {id=WavePlayer_4, properties {soundWaveAssetPtr=&tree_big_hit_01}}
 {id=WavePlayer_5, properties {soundWaveAssetPtr=&tree_big_hit_07}}
 {id=WavePlayer_6, properties {soundWaveAssetPtr=&tree_big_hit_01}}
]
properties {
 maxDistance=2097152
 soundClassObject=&Fort_Impacts
 volumeMultiplier=0.34
 concurrencyOverrides {resolutionRule=PreventNew}
 concurrencySet [&CONC_Object_Impacts_FourMax]
}
