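# Sound cue for impacts on a snowy pine tree, played in the Fort_Impacts sound class with a four-max object impact concurrency set that prevents new sounds.
# A quality-level root selects between four branches of decreasing complexity.
# The first mixes random pine hits with random snow impacts, the second mixes random hits with one fixed snow wave, and the last two play a single optimized wave.
# Modulators randomize pitch throughout.
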SoundCue /Clyde/Sounds/Fort_Impact_Sounds/Demolition/Wood/Demo_PineTree_Snow_Impact_Cue
root=QualityLevel_0
nodes [
 {id=Mixer_0, children [Modulator_1,Modulator_3], properties {inputVolume [0.5,1.25]}}
 {id=Mixer_1, children [Modulator_2,Modulator_0], properties {inputVolume [1,1]}}
 {id=Modulator_0, children [Random_0], properties {pitchMax=1.07, pitchMin=0.92}}
 {id=Modulator_1, children [Random_1], properties {pitchMax=1.07, pitchMin=0.92}}
 {id=Modulator_2, children [Random_8], properties {pitchMax=1.07, pitchMin=0.92}}
 {id=Modulator_3, children [WavePlayer_9], properties {pitchMax=1.07, pitchMin=0.92}}
 {id=Modulator_4, children [WavePlayer_15], properties {pitchMax=1.07, pitchMin=0.88}}
 {id=Modulator_5, children [WavePlayer_16], properties {pitchMax=1.07, pitchMin=0.88}}
 {id=QualityLevel_0, children [Mixer_1,Mixer_0,Modulator_4,Modulator_5]}
 {id=Random_0, children [WavePlayer_0,WavePlayer_1,WavePlayer_2,WavePlayer_3,WavePlayer_4,WavePlayer_5], properties {weights [1,1,1,1,1,1]}}
 {id=Random_1, children [WavePlayer_6,WavePlayer_7,WavePlayer_8], properties {weights [1,1,1]}}
 {id=Random_8, children [WavePlayer_44, WavePlayer_45, WavePlayer_46], properties {weights [1,1,1]}}
 {id=WavePlayer_0, properties {soundWaveAssetPtr=&/Trees_Holiday_Winter_Snow_Impacts_2017-001}}
 {id=WavePlayer_1, properties {soundWaveAssetPtr=&/Trees_Holiday_Winter_Snow_Impacts_2017-002}}
 {id=WavePlayer_15, properties {soundWaveAssetPtr=&/PineTree_Impact_Snow_opt}}
 {id=WavePlayer_16, properties {soundWaveAssetPtr=&PineTree_Impact_Snow_opt}}
 {id=WavePlayer_2, properties {soundWaveAssetPtr=&/Trees_Holiday_Winter_Snow_Impacts_2017-003}}
 {id=WavePlayer_3, properties {soundWaveAssetPtr=&/Trees_Holiday_Winter_Snow_Impacts_2017-004}}
 {id=WavePlayer_4, properties {soundWaveAssetPtr=&/Trees_Holiday_Winter_Snow_Impacts_2017-005}}
 {id=WavePlayer_44, properties {soundWaveAssetPtr=&/tree_pine_hit_01}}
 {id=WavePlayer_45, properties {soundWaveAssetPtr=&/tree_pine_hit_02}}
 {id=WavePlayer_46, properties {soundWaveAssetPtr=&/tree_pine_hit_03}}
 {id=WavePlayer_5, properties {soundWaveAssetPtr=&/Trees_Holiday_Winter_Snow_Impacts_2017-006}}
 {id=WavePlayer_6, properties {soundWaveAssetPtr=&tree_pine_hit_01}}
 {id=WavePlayer_7, properties {soundWaveAssetPtr=&tree_pine_hit_02}}
 {id=WavePlayer_8, properties {soundWaveAssetPtr=&tree_pine_hit_03}}
 {id=WavePlayer_9, properties {soundWaveAssetPtr=&Trees_Holiday_Winter_Snow_Impacts_2017-001}}
]
properties {
 maxDistance=2097152
 soundClassObject=&Fort_Impacts
 volumeMultiplier=0.34
 concurrencyOverrides {resolutionRule=PreventNew}
 concurrencySet [&CONC_Object_Impacts_FourMax]
}
