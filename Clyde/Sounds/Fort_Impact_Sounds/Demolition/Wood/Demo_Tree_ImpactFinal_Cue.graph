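# Sound cue for the final impact when a tree is demolished, routed to the Fort_Impacts_PSM sound class.
# A quality-level switch picks between a full mix of random big tree hits and random final wood layers, a reduced mix with one fixed hit, and a single optimized wave for the two remaining levels.
# Modulators vary pitch on every layer, and the CONC_Object_Impacts_FourMax concurrency set limits simultaneous impacts.
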SoundCue /Clyde/Sounds/Fort_Impact_Sounds/Demolition/Wood/Demo_Tree_ImpactFinal_Cue
root=QualityLevel_0
nodes [
 {id=Mixer_0, children [Modulator_1,Modulator_3], properties {inputVolume [1,0.8]}}
 {id=Mixer_10, children [Modulator_2,Modulator_12], properties {inputVolume [1,0.8]}}
 {id=Modulator_0, children [WavePlayer_1], properties {pitchMax=1.1, pitchMin=0.9}}
 {id=Modulator_1, children [WavePlayer_2], properties {pitchMax=1.1, pitchMin=0.9}}
 {id=Modulator_12, children [Random_14], properties {pitchMax=1.1, pitchMin=0.9}}
 {id=Modulator_2, children [Random_2], properties {pitchMax=1.07, pitchMin=0.92}}
 {id=Modulator_3, children [Random_1], properties {pitchMax=1.1, pitchMin=0.9}}
 {id=QualityLevel_0, children [Mixer_10,Mixer_0,Modulator_0,Modulator_0]}
 {id=Random_1, children [WavePlayer_6,WavePlayer_7], properties {weights [1,1]}}
 {id=Random_14, children [WavePlayer_39, WavePlayer_40], properties {weights [1,1]}}
 {id=Random_2, children [WavePlayer_15, WavePlayer_18, WavePlayer_17, WavePlayer_19], properties {weights [1,1,1,1]}}
 {id=WavePlayer_1, properties {soundWaveAssetPtr=&/Tree_ImpactFinal_opt}}
 {id=WavePlayer_15, properties {soundWaveAssetPtr=&/tree_big_hit_01}}
 {id=WavePlayer_17, properties {soundWaveAssetPtr=&/tree_big_hit_03}}
 {id=WavePlayer_18, properties {soundWaveAssetPtr=&/tree_big_hit_07}}
 {id=WavePlayer_19, properties {soundWaveAssetPtr=&/tree_big_hit_08}}
 {id=WavePlayer_2, properties {soundWaveAssetPtr=&tree_big_hit_01}}
 {id=WavePlayer_39, properties {soundWaveAssetPtr=/Game/Sounds/Fort_Impact_Sounds/Demolition/Wood/final_wood_new_03}}
 {id=WavePlayer_40, properties {soundWaveAssetPtr=/Game/Sounds/Fort_Impact_Sounds/Demolition/Wood/final_wood_new_02}}
 {id=WavePlayer_6, properties {soundWaveAssetPtr=&final_wood_new_03}}
 {id=WavePlayer_7, properties {soundWaveAssetPtr=&final_wood_new_02}}
]
properties {
 maxDistance=2097152
 soundClassObject=&Fort_Impacts_PSM
 volumeMultiplier=0.854
 concurrencyOverrides {resolutionRule=PreventNew}
 concurrencySet [&CONC_Object_Impacts_FourMax]
}
