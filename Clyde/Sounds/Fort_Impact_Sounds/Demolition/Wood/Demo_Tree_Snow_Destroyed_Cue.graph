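# Sound cue for a snow-covered tree being destroyed, played on the impacts sound class under a four-max concurrency set that blocks new sounds.
# A quality switch picks a full mix, a leaner mix with fewer variations, or one optimized destroy wave for the two remaining levels.
# Both mixes layer a tree impact faded by a volume envelope, wood debris and snow impacts, with random pitch on every layer.
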SoundCue /Clyde/Sounds/Fort_Impact_Sounds/Demolition/Wood/Demo_Tree_Snow_Destroyed_Cue
root=QualityLevel_0
nodes [
 {
  id=Enveloper_0
  children [Modulator_3]
  properties {
   pitchCurve {editorCurveData {keys [{time=0}]}}
   volumeCurve {editorCurveData {keys [<at 0 1>, {arriveTangent=-0.268, interpMode=RCIM_Cubic, leaveTangent=-0.268, time=1.358, value=0.9856}, <at 3.8687 -0.0368 RCIM_Cubic>]}}
  }
 }
 {
  id=Enveloper_3
  children [Modulator_2]
  properties {
   pitchCurve {editorCurveData {keys [{time=0}]}}
   volumeCurve {editorCurveData {keys [<at 0 1>, {arriveTangent=-0.268, interpMode=RCIM_Cubic, leaveTangent=-0.268, time=1.358, value=0.9856}, <at 3.8687 -0.0368 RCIM_Cubic>]}}
  }
 }
 {id=Mixer_0, children [Enveloper_0,Modulator_4,Modulator_5], properties {inputVolume [1,1,1]}}
 {id=Mixer_9, children [Enveloper_3,Modulator_11,Modulator_0], properties {inputVolume [1,1,1]}}
 {id=Modulator_0, children [Random_0], properties {pitchMax=1.07, pitchMin=0.92}}
 {id=Modulator_1, children [WavePlayer_6], properties {pitchMax=1.1, pitchMin=0.9}}
 {id=Modulator_11, children [Random_13], properties {pitchMax=1.1, pitchMin=0.9}}
 {id=Modulator_2, children [Random_7], properties {pitchMax=1.07, pitchMin=0.92}}
 {id=Modulator_3, children [WavePlayer_7], properties {pitchMax=1.07, pitchMin=0.92}}
 {id=Modulator_4, children [Random_2], properties {pitchMax=1.1, pitchMin=0.9}}
 {id=Modulator_5, children [WavePlayer_11], properties {pitchMax=1.07, pitchMin=0.92}}
 {id=QualityLevel_0, children [Mixer_9,Mixer_0,Modulator_1,Modulator_1]}
 {id=Random_0, children [WavePlayer_0,WavePlayer_1,WavePlayer_2,WavePlayer_3,WavePlayer_4,WavePlayer_5], properties {weights [1,1,1,1,1,1]}}
 {id=Random_13, children [WavePlayer_37, WavePlayer_38], properties {weights [1,1]}}
 {id=Random_2, children [WavePlayer_9, WavePlayer_10], properties {bRandomizeWithoutReplacement=false, weights [1,1]}}
 {id=Random_7, children [WavePlayer_42, WavePlayer_43], properties {weights [1,1]}}
 {id=WavePlayer_0, properties {soundWaveAssetPtr=&/Trees_Holiday_Winter_Snow_Impacts_2017-001}}
 {id=WavePlayer_1, properties {soundWaveAssetPtr=&/Trees_Holiday_Winter_Snow_Impacts_2017-002}}
 {id=WavePlayer_10, properties {soundWaveAssetPtr=/Game/Sounds/Fort_Impact_Sounds/Demolition/Wood/final_wood_new_02}}
 {id=WavePlayer_11, properties {soundWaveAssetPtr=&Trees_Holiday_Winter_Snow_Impacts_2017-001}}
 {id=WavePlayer_2, properties {soundWaveAssetPtr=&/Trees_Holiday_Winter_Snow_Impacts_2017-003}}
 {id=WavePlayer_3, properties {soundWaveAssetPtr=&/Trees_Holiday_Winter_Snow_Impacts_2017-004}}
 {id=WavePlayer_37, properties {soundWaveAssetPtr=/Game/Sounds/Fort_Impact_Sounds/Demolition/Wood/final_wood_new_03}}
 {id=WavePlayer_38, properties {soundWaveAssetPtr=&final_wood_new_02}}
 {id=WavePlayer_4, properties {soundWaveAssetPtr=&/Trees_Holiday_Winter_Snow_Impacts_2017-005}}
 {id=WavePlayer_42, properties {soundWaveAssetPtr=/Clyde/Sounds/Fort_Impact_Sounds/Demolition/Fort_Impact_TreeFinal_02}}
 {id=WavePlayer_43, properties {soundWaveAssetPtr=/Clyde/Sounds/Fort_Impact_Sounds/Demolition/Fort_Impact_TreeFinal_01}}
 {id=WavePlayer_5, properties {soundWaveAssetPtr=&/Trees_Holiday_Winter_Snow_Impacts_2017-006}}
 {id=WavePlayer_6, properties {soundWaveAssetPtr=&/Tree_Destroy_opt}}
 {id=WavePlayer_7, properties {soundWaveAssetPtr=&Fort_Impact_TreeFinal_02}}
 {id=WavePlayer_9, properties {soundWaveAssetPtr=&final_wood_new_03}}
]
properties {
 maxDistance=2097152
 soundClassObject=&Fort_Impacts
 volumeMultiplier=0.272
 concurrencyOverrides {resolutionRule=PreventNew}
 concurrencySet [&CONC_Object_Impacts_FourMax]
}
