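# Sound cue for the final impact of demolishing a wooden fence.
# A mixer layers a random wood fence impact with a quieter random explosion layer, each chosen from three waves.
# Each layer has its own random pitch modulation, with the explosion pitched lower to sit beneath the wood.
# The whole cue is pitched down and attenuated in volume, routed to an impacts sound class and assigned an object-impact concurrency set.
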SoundCue /Clyde/Sounds/Fort_Impact_Sounds/Demolition/Wood/Demo_WoodFence_ImpactFinal_Cue
root=Mixer_2
nodes [
 {id=Mixer_2, children [Modulator_2,Modulator_3], properties {inputVolume [1,0.7]}}
 {id=Modulator_2, children [Random_1], properties {pitchMax=1.07, pitchMin=0.92}}
 {id=Modulator_3, children [Random_4], properties {pitchMax=0.95, pitchMin=0.85}}
 {id=Random_1, children [WavePlayer_6,WavePlayer_8,WavePlayer_9], properties {weights [1,1,1]}}
 {id=Random_4, children [WavePlayer_18, WavePlayer_19, WavePlayer_20], properties {weights [1,1,1]}}
 {id=WavePlayer_18, properties {soundWaveAssetPtr=/Clyde/Sounds/Fort_Impact_Sounds/Demolition/Explo_Impact_Layer01}}
 {id=WavePlayer_19, properties {soundWaveAssetPtr=/Clyde/Sounds/Fort_Impact_Sounds/Demolition/Explo_Impact_Layer02}}
 {id=WavePlayer_20, properties {soundWaveAssetPtr=/Clyde/Sounds/Fort_Impact_Sounds/Demolition/Explo_Impact_Layer03}}
 {id=WavePlayer_6, properties {soundWaveAssetPtr=/Clyde/Sounds/Fort_Impact_Sounds/Demolition/Wood_Fence_ImpactFinal02}}
 {id=WavePlayer_8, properties {soundWaveAssetPtr=/Clyde/Sounds/Fort_Impact_Sounds/Demolition/Wood_Fence_ImpactFinal04}}
 {id=WavePlayer_9, properties {soundWaveAssetPtr=/Clyde/Sounds/Fort_Impact_Sounds/Demolition/Wood_Fence_ImpactFinal05}}
]
properties {
 maxDistance=2097152
 pitchMultiplier=0.85
 soundClassObject=&Fort_Impacts_PSM
 volumeMultiplier=0.34
 concurrencyOverrides {maxCount=3, resolutionRule=StopOldest}
 concurrencySet [&CONC_Object_Impacts_FourMax]
}
